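# Translucent, two-sided Niagara sprite material in the Party Trooper body FX, drawing an animated candy-colored radial pattern.
# UV distance, offset by a heart curve atlas row, drives time-scrolled color and star-mask lookups; the atan2 angle takes G for both inputs.
# A crack-distorted split mask, picked by dynamic parameter, multiplies opacity before depth fade, while scaled color feeds an emissive adjustment function.
# Spin, extrusion and offset branches feed nothing downstream.
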Material /BRCosmetics/Characters/Player/Male/Medium/Bodies/M_MED_Party_Trooper_A/FX/M_RadialPatternSprite
graph {
 properties {
  bCastRayTracedShadows=false
  bUsedWithNiagaraSprites=true
  blendMode=BLEND_Translucent
  pixelDepthOffsetMode=PDOM_Legacy
  translucencyLightingMode=TLM_VolumetricDirectional
  twoSided=true
  usedShadingModels=Default Lit
 }
 nodes [
  n1 = DynamicParameter() {
   parameterIndex=1
   defaultValue=<rgba 1 1 1 2>
   paramNames [Base Sprite Offset U, Base Sprite Offset V, Radial Mask Offset U, Radial Mask Offset V]
  }
  n2 = ComponentMask(None=@n1.Base Sprite Offset U) {r=true}
  n3 = ComponentMask(None=@n1.Base Sprite Offset V) {g=true}
  n4 = MaterialFunctionCall(Float2=TextureCoordinate()) {materialFunction=&BreakOutFloat2Components}
  n5 = OneMinus(None=@n4.G)
  n6 = Subtract(A=AppendVector(A=@n5, B=@n4.R)) {constB=0.5}
  n7 = SquareRoot(None=Multiply(
    A=MaterialFunctionCall(Vector 2=@n6) {materialFunction=/Engine/Functions/Engine_MaterialFunctions02/Utility/VectorLength}.V2 Length
    B=Constant() {r=0.5}
   ))
  n8 = MaterialFunctionCall(Float2=@n6) {materialFunction=&BreakOutFloat2Components}
  n9 = Divide(
   A=Add(A=Arctangent2Fast(Y=@n8.G, X=@n8.G), B=MaterialFunctionCall() {materialFunction=&Pi}.Result)
   B=MaterialFunctionCall(Multiplier=Constant() {r=2}) {materialFunction=&Pi}.Result
  )
  n10 = CurveAtlasRowParameter(CurveTime=@n9) {
   atlas=&/CA_Pattern_RadialOffset
   curve=&/C_RadialOffset_Heart_A
   defaultValue=3
   parameterName=CurveAtlas_RadialOffset
   sortPriority=44
  }
  n11 = Constant() {r=0.1}
  n12 = AppendVector(A=Add(A=Power(Base=@n7) {constExponent=0.85}, B=Multiply(A=@n10.R, B=@n11)), B=@n9)
  n13 = Constant2Vector() {g=1, r=1}
  n14 = ComponentMask(None=@n13) {g=true}
  n15 = DynamicParameter() {defaultValue=<rgba 1 1 1 2>, paramNames [Sprite Extrusion Speed, Sprite Spin Speed, Radial Mask Strength, Radial Mask Power]}
  Add(
   A=AppendVector(A=Time() {bIgnorePause=true, bOverride_Period=true, period=3}, B=Multiply(A=@n14, B=@n15.Sprite Spin Speed))
   B=Add(A=AppendVector(A=@n2, B=@n3), B=@n12)
  )
  AppendVector(A=ComponentMask(None=@n1.Radial Mask Offset U) {r=true}, B=ComponentMask(None=@n1.Radial Mask Offset V) {g=true})
  n16 = MaterialFunctionCall(Float2=@n12) {materialFunction=&BreakOutFloat2Components}
  n17 = Multiply(A=Max(A=@n16.R) {constB=0.25}, B=ScalarParameter() {defaultValue=2.5, parameterName=Repeat_Frequency, sliderMax=5}) {constB=2.5}
  n18 = Multiply(A=Time() {bIgnorePause=true, period=1}, B=ScalarParameter() {defaultValue=1, parameterName=RadialSpeed, sliderMax=1, sortPriority=2}) {constB=0.1}
  n19 = Constant() {r=1}
  n20 = CurveAtlasRowParameter(CurveTime=Fmod(A=Add(A=@n17, B=@n18), B=@n19)) {
   atlas=&CA_Pattern_RadialOffset
   curve=&/C_CandyColored_A
   defaultValue=5
   parameterName=CurveAtlas_Color
  }
  n21 = CurveAtlasRowParameter(CurveTime=Fmod(A=@n16.R, B=@n19)) {
   atlas=&CA_Pattern_RadialOffset
   curve=&/C_Star_Mask
   defaultValue=4
   parameterName=CurveAtlas_Mask
   sortPriority=31
  }
  n22 = DynamicParameter() {parameterIndex=3, defaultValue=<rgba 0 0 0 2>, paramNames [Depth Fade, N/A, N/A, Eye CrackPiece Picker]}
  n23 = Constant()
  n24 = CurveAtlasRowParameter(CurveTime=Multiply(A=@n5, B=Constant() {r=0.7})) {
   atlas=&/CA_LinearCracks
   curve=&/C_LinOffset_Crack
   defaultValue=1
   parameterName=CurveAtlas_CraskMask
   sortPriority=233
  }
  n25 = LinearInterpolate(A=@n4.R, B=@n24.R) {constAlpha=0.4}
  n26 = ScalarParameter() {parameterName=Shift_Crack}
  n27 = CurveAtlasRowParameter(CurveTime=LinearInterpolate(A=@n4.R, B=Add(A=@n25, B=@n26), Alpha=@n22.N/A) {constAlpha=0}) {
   atlas=&CA_LinearCracks
   bUseCustomPrimitiveData=true
   curve=&/C_SplitCenter_Mask
   defaultValue=1
   parameterName=CurveAtlas_CraskMask
   sortPriority=233
  }
  n28 = OneMinus(None=@n27)
  DepthFade(
   Opacity=Multiply(A=Multiply(A=@n20.A, B=@n21), B=If(A=@n22.Eye CrackPiece Picker, B=@n23, "A > B"=@n27, "A == B"=@n28, "A < B"=@n28))
   FadeDistance=@n22.Depth Fade
  )
  n29 = ScalarParameter() {defaultValue=0.4, parameterName=ScaleColor, sliderMax=15}
  MaterialFunctionCall(Emissive=Saturate(None=Multiply(A=@n20, B=@n29))) {materialFunction=&MF_FX_EmissiveAdjustments}
  Multiply(A=ComponentMask(None=@n13) {r=true}, B=@n15.Sprite Extrusion Speed)
  ScalarParameter() {parameterName=CrackEdgeWeight}
  ScalarParameter() {parameterName=Switch_LR, sliderMax=1, sortPriority=2}
 ]
}
