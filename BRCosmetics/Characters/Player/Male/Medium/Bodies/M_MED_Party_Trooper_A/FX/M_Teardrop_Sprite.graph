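# Unlit translucent Niagara sprite material for Party Trooper FX, drawing a procedural teardrop with animated candy colors.
# Center-relative UV distance is reshaped by a curve-atlas offset, whose arctangent input uses one UV component for both axes, blended by Drop Sag.
# That distance samples a star mask curve and, scrolled by time, a color curve.
# Color feeds emissive adjustments; color alpha times mask feeds depth fade, with sag and fade driven per particle.
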Material /BRCosmetics/Characters/Player/Male/Medium/Bodies/M_MED_Party_Trooper_A/FX/M_Teardrop_Sprite
graph {
 properties {
  bCastRayTracedShadows=false
  bUsedWithNiagaraSprites=true
  blendMode=BLEND_Translucent
  pixelDepthOffsetMode=PDOM_Legacy
  shadingModel=MSM_Unlit
  usedShadingModels=Unlit
 }
 nodes [
  n1 = MaterialFunctionCall(Float2=TextureCoordinate()) {materialFunction=&BreakOutFloat2Components}
  n2 = OneMinus(None=@n1.G)
  n3 = Add(A=@n2) {constB=0.2}
  n4 = DynamicParameter() {parameterIndex=3, paramNames [Depth Fade, N/A, N/A, Drop Sag]}
  n5 = Multiply(A=ScalarParameter() {defaultValue=1, parameterName=Drop_Sag, sliderMax=1}, B=@n4.Drop Sag)
  n6 = Subtract(A=AppendVector(A=LinearInterpolate(A=@n2, B=@n3, Alpha=@n5) {constAlpha=1}, B=@n1.R)) {constB=0.5}
  n7 = SquareRoot(None=Multiply(
    A=MaterialFunctionCall(Vector 2=@n6) {materialFunction=/Engine/Functions/Engine_MaterialFunctions02/Utility/VectorLength}.V2 Length
    B=Constant() {r=0.8}
   ))
  n8 = MaterialFunctionCall(Float2=@n6) {materialFunction=&BreakOutFloat2Components}
  n9 = Arctangent2Fast(Y=@n8.G, X=@n8.G)
  n10 = MaterialFunctionCall() {materialFunction=&Pi}
  n11 = Constant() {r=2}
  n12 = CurveAtlasRowParameter(CurveTime=Divide(
    A=Add(A=@n9, B=@n10.Result)
    B=MaterialFunctionCall(Multiplier=@n11) {materialFunction=&Pi}.Result
   )) {
   atlas=&/CA_Pattern_RadialOffset
   curve=&/C_RadialOffset_TearDrop
   defaultValue=7
   parameterName=CurveAtlas_offset
   sortPriority=12
  }
  n13 = Subtract(A=@n12.R, B=Constant() {r=0.55})
  n14 = Constant() {r=0.3}
  n15 = LinearInterpolate(A=Power(Base=@n7) {constExponent=0.85}, B=Add(A=@n7, B=Multiply(A=@n13, B=@n14)), Alpha=@n5) {constAlpha=1}
  n16 = Multiply(A=Max(A=@n15) {constB=0.2})
  n17 = Multiply(
   A=OneMinus(None=Time() {bIgnorePause=true, bOverride_Period=true, period=1})
   B=ScalarParameter() {defaultValue=1, parameterName=RadialSpeed, sliderMax=1, sortPriority=2}
  ) {constB=0.1}
  n18 = Constant() {r=1}
  n19 = CurveAtlasRowParameter(CurveTime=Fmod(A=Add(A=@n16, B=@n17), B=@n18)) {
   atlas=&CA_Pattern_RadialOffset
   curve=&/C_CandyColored_A
   defaultValue=5
   parameterName=CurveAtlas_Color
  }
  n20 = CurveAtlasRowParameter(CurveTime=Fmod(A=@n15, B=@n18)) {
   atlas=&CA_Pattern_RadialOffset
   curve=&/C_Star_Mask
   defaultValue=4
   parameterName=CurveAtlas_Mask
   sortPriority=31
  }
  DepthFade(Opacity=Multiply(A=@n19.A, B=@n20), FadeDistance=@n4.Depth Fade)
  MaterialFunctionCall(Emissive=@n19) {materialFunction=&MF_FX_EmissiveAdjustments}
 ]
}
