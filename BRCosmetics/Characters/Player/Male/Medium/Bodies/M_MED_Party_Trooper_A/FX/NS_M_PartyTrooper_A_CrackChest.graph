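# Niagara effect for the Party Trooper A body cosmetic, emitting rising, curl-noise glow particles from the skinned mesh; only the GPU Body_Particles emitter is enabled.
# Spawns sample the Party Trooper Buffet emissive texture at mesh UVs, culling by threshold and taking its color.
# A region mask sized by hand/foot-hidden user flags drives particle alpha.
# Disabled cracked chest-heart color, outline and drip emitters follow sockets and read User.Damage.
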NiagaraSystem /BRCosmetics/Characters/Player/Male/Medium/Bodies/M_MED_Party_Trooper_A/FX/NS_M_PartyTrooper_A_CrackChest
dependencies {
 usedDataInterfaces [&NiagaraDataInterfaceCurve, /Script/Niagara.NiagaraDataInterfaceTexture, &NiagaraDataInterfaceArrayFloat]
 usedDynamicInputs []
 usedModules [&SystemState, &Timeline]
}
emitters {
 Body_Particles {
  simTarget=GPUComputeSim
  emitterUpdateScript [
   &EmitterState(Life Cycle Mode=::NewEnumerator1, Loop Duration=15, UseLoopDelay=false)
   &SpawnRate(SpawnRate=400, Use Spawn Probability=false)
  ]
  particleSpawnScript [
   &InitializeParticle(
    Lifetime Max=1
    Lifetime Min=0.2
    Lifetime Mode=::NewEnumerator1
    Sprite Size Mode=::NewEnumerator3
    Uniform Sprite Size=&RandomRangeFloat(Minimum=7, Maximum=25)
    UsePositionOffset=false
    Write Material Random=true
   )
   &SkeletalMeshLocation(
    MaskSpawnGroup=false
    Mesh Sampling Type=::NewEnumerator7
    OffsetAlongNormal=false
    Orientation Sampling=::NewEnumerator1
    OverrideRandomSeed=false
    Randomness Mode:ENiagaraRandomnessMode=::NewEnumerator1
    Skeletal Mesh=$User.SkelMeshBody
   )
   /Niagara/Modules/Update/Textures/SampleTexture(
    Texture=DataInterface() {propertyValues {
      Texture=/BRCosmetics/Characters/Player/Male/Medium/Bodies/M_MED_Soldier_Redux/Skins/PartyTrooperBuffet/Textures/T_M_MED_Party_Trooper_Buffet_Body_E
      TextureUserParameter {parameter {name=None, typeDefHandle {registeredTypeIndex=2}, varData []}}
     }}
    UV:Vector2f=$Particles.SkeletalMeshLocation.SampledUV
   ) {module=SampleTexture001}
   &KillParticles(Kill Particles=&SetBoolByFloatComparison(
     A=&MakeFloatFromLinearColor(LinearColor=$Particles.SampleTexture001.SampledColor, Channel:ENiagaraLinearColor_Channels=::NewEnumerator0)
     B=0.2
     Comparison Type=::NewEnumerator6
    )) {module=KillParticles001}
   /SetVariables(Particles.Color=$Particles.SampleTexture001.SampledColor) {isSetParameters=true}
   &AddVelocity(Velocity=&Multiply_VectorByFloat(Vector=$Particles.SkeletalMeshLocation.SampledNormal, Float=15))
  ]
  particleUpdateScript [
   &ParticleState()
   &ScaleSpriteSize(Scale Factor=&Vector2DFromFloat(Value=&FloatFromCurve(
      FloatCurve=DataInterface() {propertyValues {Curve {keys [
          {arriveTangent=4.1465, interpMode=RCIM_Cubic, leaveTangent=4.1465, tangentMode=RCTM_User, time=0, value=0.15}
          <at 0.2 1>
          {arriveTangent=-0.6938, interpMode=RCIM_Cubic, leaveTangent=-0.6938, tangentMode=RCTM_User, time=0.5, value=0.35}
          <at 1.3 0>
         ]}}}
      Scale Curve=3
     )))
   &RegionMask(
    Invert Region=true
    Offset:Vector3f=&MakeCustomVectorFromBool(Bool=$User.bisFootHidden, True Vector=<0 0 -72>, False Vector=<0 0 -100>)
    Offset Space:ENiagaraCoordinateSpace=::World
    Region Type:ENiagaraDebugDrawShape=::NewEnumerator2
    Rotation:Vector3f=<0 0.25 0>
    UseOffset=true
    UseRotation=false
   ) {enabled=false}
   &RegionMask(
    Offset Space:ENiagaraCoordinateSpace=::World
    Region Dimensions:Vector3f=&MakeVector(
     X=&MakeCustomFloatFromBool(Bool=$User.bisHandHidden, True Float=60, False Float=110)
     Y=&MakeCustomFloatFromBool(Bool=$User.bisHandHidden, True Float=60, False Float=110)
     Z=&MakeCustomFloatFromBool(Bool=$User.bisFootHidden, True Float=280, False Float=500)
    )
    Region Type:ENiagaraDebugDrawShape=::NewEnumerator1
    Rotation:Vector3f=<0 0.25 0>
    UseOffset=false
    UseRotation=false
   ) {module=RegionMask001}
   &ScaleColor(
    Scale Alpha=&FloatFromCurve(
     FloatCurve=DataInterface() {propertyValues {Curve {keys [<at 0 0.1 RCIM_Cubic>,<at 0.2 0.9>,<at 1.0011 0.002 RCIM_Cubic>]}}}
     Scale Curve=&Lerp_Float(A=0, B=0.5, Alpha=$Particles.RegionMask001.Float)
    )
    Scale RGB=&Multiply_VectorByFloat(Vector=<1 1 1>, Float=0.5)
    ScaleA=true
    ScaleRGB=true
   )
   &ScaleColor(
    Scale Alpha=&FloatFromCurve(
     FloatCurve=DataInterface() {propertyValues {Curve {keys [<at 0 0.1 RCIM_Cubic>,<at 0.2 0.9>,<at 1.0011 0.002 RCIM_Cubic>]}}}
     Scale Curve=0.5
    )
    Scale RGB=&Multiply_VectorByFloat(Vector=<1 1 1>, Float=0.5)
    ScaleA=true
    ScaleRGB=true
   ) {enabled=false, module=ScaleColor001}
   &GravityForce(Gravity=<0 0 50>)
   &Drag(
    Drag=&RandomRangeFloat(Minimum=0.1, Maximum=1, Evaluation Type=::NewEnumerator1)
    Use Linear Drag=true
    Use Rotational Drag=false
   )
   &CurlNoiseForce(Noise Frequency=125, Noise Strength=15, Pan Noise=false)
   &InheritVelocity(
    Inherited Velocity Amount Scale=<1 1 0.3>
    Inherited Velocity Speed Limit:NiagaraFloat=&RandomRangeFloat(Minimum=10, Maximum=15)
    Speed Limit=true
    UseSpeedThreshold=true
   )
   &SolveForcesAndVelocity(Clamp Velocity=false, Limit Acceleration=false)
   &ScaleSpriteSizeBySpeed(
    Max Scale Factor=<1.2 3.3>
    Min Scale Factor=<0.75 0.3>
    Sample Scale Factor By Curve=true
    Scale Factor Curve=DataInterface() {propertyValues {Curve {keys [
        {arriveTangent=2.865, interpMode=RCIM_Cubic, time=0.1247, value=-0.0006}
        {arriveTangent=4.1522, interpMode=RCIM_Cubic, leaveTangent=4.1522, tangentMode=RCTM_User, time=0.3948, value=0.4893}
        <at 0.9958 0.9974 RCIM_Cubic>
       ]}}}
    Velocity Threshold=1500
   ) {enabled=false}
   &CameraOffset(Camera Offset Amount=15)
  ]
  rendererClasses [&NiagaraSpriteRendererProperties]
  renderers [{rendererClass=&NiagaraSpriteRendererProperties}]
 }
 ChestHeart_Cracked_color {
  bEnabled=false
  emitterUpdateScript [
   &EmitterState()
   &SpawnBurst_Instantaneous(Spawn Count=2, Use Spawn Probability=false)
  ]
  particleSpawnScript [
   &InitializeParticle(
    Color Mode=::NewEnumerator0
    Lifetime=15.1
    Position Offset=<0 0 125>
    Sprite Rotation Mode=::NewEnumerator0
    Sprite Size=<15 13>
    Sprite Size Mode=::NewEnumerator4
    UsePositionOffset=false
    Write Material Random=true
   )
   &self:Init_ChestOffsets(EyeCrackAngle=-8, EyeCrackShift=0.3) {module=Init_ChestOffsets}
   &SampleSkeletalMesh(
    AdditionalYawPitchRoll=false
    Bone / Socket Index:NiagaraInt32=$Particles.EyeSocketIdx
    Mesh Sampling Type:ENiagaraSkelSamplingModeFull=::NewEnumerator5
    Skeletal Mesh:NiagaraDataInterfaceSkeletalMesh=$User.SkelMeshBody
    Socket Sampling Mode:ENiagaraSocketSamplingMode=::NewEnumerator5
   ) {enabled=false}
  ]
  particleUpdateScript [
   &ParticleState()
   &SkeletalMeshLocation(
    "Interpolate Mesh Sampling (Update)"=true
    MaskSpawnGroup=false
    Mesh Sampling Type=::NewEnumerator5
    Orientation Sampling=::NewEnumerator1
    Sampled Position Offset Coordinate Space:ENiagaraSkelMeshTransforms=::NewEnumerator0
    Skeletal Mesh=$User.SkelMeshBody
    Socket Sampling Mode=::NewEnumerator5
   )
   &self:Update_ChestOffsets(
    Eye Halves Tilt Z:NiagaraFloat=&Lerp_Float(A=-8, B=-8, Alpha=$System.Timeline.NormalizedTime)
    EyeCrackAngle=-8
    EyeCrackShift=1.3
    EyeIsCracked:NiagaraFloat=$User.Damage
    EyeOffset:Vector3f=<0 -2 0>
    Eyes LR Tilt Y=0
    Eyes LR Tilt Z=0
    SM_SampledRotation:Quat4f=$Particles.SkeletalMeshLocation.SampledRotation
   ) {module=Update_ChestOffsets}
   /SetVariables(Particles.SpriteSize=&Multiply_Vector2D_ByFloat(
     Float=&Multiply_Float(
      A=&FloatFromCurve(
       FloatCurve=DataInterface() {propertyValues {Curve {keys [
           {interpMode=RCIM_Cubic, tangentMode=RCTM_User, time=0, value=0.975}
           {arriveTangent=2.8837, interpMode=RCIM_Cubic, leaveTangent=2.8837, tangentMode=RCTM_User, time=0.0004, value=1.4938}
           {arriveTangent=-3.0766, interpMode=RCIM_Cubic, leaveTangent=-3.0766, tangentMode=RCTM_User, time=0.3, value=1.5}
           {arriveTangent=-0.1031, interpMode=RCIM_Cubic, leaveTangent=-0.1031, tangentMode=RCTM_User, time=0.6, value=0.6}
           <at 0.8 0.95>
           {tangentMode=RCTM_User, time=1, value=0.975}
          ]}}}
       CurveIndex=$System.Timeline.NormalizedTime
      )
      B=1
     )
     Vector2D=$Particles.Initial.SpriteSize
    )) {isSetParameters=true, module=SetVariables_#1}
   /SetVariables(Particles.SpriteSize=&Multiply_Vector2D_ByFloat(Float=HlslExpression() {hlslExpression=.95}, Vector2D=$Particles.SpriteSize)) {isSetParameters=true, module=SetVariables_#2}
   &CameraOffset(Camera Offset Amount=12)
  ]
  rendererClasses [&NiagaraSpriteRendererProperties]
  renderers [{rendererClass=&NiagaraSpriteRendererProperties}]
 }
 ChestHeart_Cracked_outline {
  bEnabled=false
  emitterUpdateScript [
   &EmitterState()
   &SpawnBurst_Instantaneous(Spawn Count=2, Use Spawn Probability=false)
  ]
  particleSpawnScript [
   &InitializeParticle(
    Color Mode=::NewEnumerator0
    Lifetime=15.1
    Position Offset=<0 0 125>
    Sprite Rotation Mode=::NewEnumerator0
    Sprite Size=<15 13>
    Sprite Size Mode=::NewEnumerator4
    UsePositionOffset=false
    Write Material Random=true
   )
   &self:Init_ChestOffsets(EyeCrackAngle=-8, EyeCrackShift=0.3) {module=Init_ChestOffsets}
   &SampleSkeletalMesh(
    AdditionalYawPitchRoll=false
    Bone / Socket Index:NiagaraInt32=$Particles.EyeSocketIdx
    Mesh Sampling Type:ENiagaraSkelSamplingModeFull=::NewEnumerator5
    Skeletal Mesh:NiagaraDataInterfaceSkeletalMesh=$User.SkelMeshBody
    Socket Sampling Mode:ENiagaraSocketSamplingMode=::NewEnumerator5
   ) {enabled=false}
  ]
  particleUpdateScript [
   &ParticleState()
   &SkeletalMeshLocation(
    "Interpolate Mesh Sampling (Update)"=true
    MaskSpawnGroup=false
    Mesh Sampling Type=::NewEnumerator5
    Orientation Sampling=::NewEnumerator1
    Sampled Position Offset Coordinate Space:ENiagaraSkelMeshTransforms=::NewEnumerator2
    Skeletal Mesh=$User.SkelMeshBody
    Socket Sampling Mode=::NewEnumerator5
   )
   &self:Update_ChestOffsets(
    Eye Halves Tilt Z:NiagaraFloat=&Lerp_Float(A=-8, B=-8, Alpha=$User.Damage)
    EyeCrackAngle=-12
    EyeCrackShift=0.3
    EyeIsCracked:NiagaraFloat=$User.Damage
    EyeOffset:Vector3f=<0 -3 0>
    Eyes LR Tilt Y=0
    Eyes LR Tilt Z=0
    SM_SampledRotation:Quat4f=$Particles.SkeletalMeshLocation.SampledRotation
   ) {module=Update_ChestOffsets}
   /SetVariables(Particles.SpriteSize=&Multiply_Vector2D_ByFloat(
     Float=&Multiply_Float(
      A=&FloatFromCurve(
       FloatCurve=DataInterface() {propertyValues {Curve {keys [
           {interpMode=RCIM_Cubic, tangentMode=RCTM_User, time=0, value=0.975}
           {arriveTangent=2.8837, interpMode=RCIM_Cubic, leaveTangent=2.8837, tangentMode=RCTM_User, time=0.0004, value=1.4938}
           {arriveTangent=-3.0766, interpMode=RCIM_Cubic, leaveTangent=-3.0766, tangentMode=RCTM_User, time=0.3, value=1.5}
           {arriveTangent=-0.1031, interpMode=RCIM_Cubic, leaveTangent=-0.1031, tangentMode=RCTM_User, time=0.6, value=0.6}
           <at 0.8 0.95>
           {tangentMode=RCTM_User, time=1, value=0.975}
          ]}}}
       CurveIndex=$System.Timeline.NormalizedTime
      )
      B=1
     )
     Vector2D=$Particles.Initial.SpriteSize
    )) {isSetParameters=true, module=SetVariables_#1}
   /SetVariables(Particles.SpriteSize=&Multiply_Vector2D_ByFloat(Float=HlslExpression() {hlslExpression=1.1}, Vector2D=$Particles.SpriteSize)) {isSetParameters=true, module=SetVariables_#2}
   &CameraOffset(Camera Offset Amount=12)
  ]
  rendererClasses [&NiagaraSpriteRendererProperties]
  renderers [{rendererClass=&NiagaraSpriteRendererProperties}]
 }
 ChestHeart_Drip {
  bEnabled=false
  emitterUpdateScript [
   &EmitterState()
   &SpawnRate(SpawnRate=3, Use Spawn Probability=false) {enabled=false}
   &SpawnPerFrame(Spawn Probability:NiagaraFloat=&Multiply_Float(A=0.015, B=$User.Damage), Use Spawn Probability=true)
  ]
  particleSpawnScript [
   &InitializeParticle(
    Color Mode=::NewEnumerator0
    Lifetime=0.6
    Position Offset=<0 0 125>
    Sprite Rotation Mode=::NewEnumerator0
    Sprite Size Mode=::NewEnumerator3
    Uniform Sprite Size=4
    UsePositionOffset=false
    Write Material Random=true
   )
   &self:Set_TearOffsets_01(EyeCrackShift=1, EyeOffset:Vector3f=<3 -0.5 -4>) {module=Set_TearOffsets_01}
   &SampleSkeletalMesh(
    AdditionalYawPitchRoll=false
    Mesh Sampling Type:ENiagaraSkelSamplingModeFull=::NewEnumerator5
    Skeletal Mesh:NiagaraDataInterfaceSkeletalMesh=$User.SkelMeshBody
    Socket Sampling Mode:ENiagaraSocketSamplingMode=::NewEnumerator5
   )
   &self:Update_ChestDrip_Offsets(EyeCrackAngle=0, SM_SampledRotation:Quat4f=$Particles.SampleSkeletalMesh.SampledRotation) {module=Update_ChestDrip_Offsets}
   &SkeletalMeshLocation(
    Bone / Socket Index=$Particles.EyeSocketIdx
    "Inherit Velocity (Scale)"=0.4
    MaskSpawnGroup=false
    Mesh Sampling Type=::NewEnumerator5
    Position Sampling=::NewEnumerator1
    Sampled Position Offset=$Particles.EyeOffset
    Sampled Position Offset Coordinate Space:ENiagaraSkelMeshTransforms=::NewEnumerator0
    Skeletal Mesh=$User.SkelMeshBody
    Socket Sampling Mode=::NewEnumerator5
    Velocity Sampling=::NewEnumerator0
    VelocitySpeedLimit=false
   )
   &AddVelocity(
    Coordinate Space=::Simulation
    Velocity=&RotateVectorByQuaternion(Vector To Rotate=<0 0 0>, Quaternion=$Particles.SampleSkeletalMesh.SampledRotation)
   )
  ]
  particleUpdateScript [
   &ParticleState()
   &GravityForce(Gravity=<0 0 -120>)
   &SolveForcesAndVelocity(Clamp Velocity=false, Limit Acceleration=false)
   &DynamicMaterialParameters(
    Index 3 Param 4=&FloatFromCurve(FloatCurve=DataInterface() {propertyValues {Curve {keys [
         {arriveTangent=-6.0786, interpMode=RCIM_Cubic, leaveTangent=-6.0786, tangentMode=RCTM_User, time=0.6, value=0.85}
         {arriveTangent=-6.8417, interpMode=RCIM_Cubic, leaveTangent=-6.8417, tangentMode=RCTM_User, time=0.7, value=0.1}
        ]}}})
    Module.Param0WriteEnabled3=false
    Module.Param1WriteEnabled3=false
    Module.Param2WriteEnabled3=false
    Module.Param3WriteEnabled3=true
    Write Parameter Index 0=false
    Write Parameter Index 3=true
   )
  ]
  rendererClasses [&NiagaraSpriteRendererProperties]
  renderers [{rendererClass=&NiagaraSpriteRendererProperties}]
 }
}
summary {
 systemName=NS_M_PartyTrooper_A_CrackChest
 emitters [
  {bEnabled=false, emitterName=ChestHeart_Cracked_outline, rendererClasses [&NiagaraSpriteRendererProperties]}
  {bEnabled=false, emitterName=ChestHeart_Drip, rendererClasses [&NiagaraSpriteRendererProperties]}
  {bEnabled=false, emitterName=ChestHeart_Cracked_color, rendererClasses [&NiagaraSpriteRendererProperties]}
  {emitterName=Body_Particles, simTarget=GPUComputeSim, rendererClasses [&NiagaraSpriteRendererProperties]}
 ]
 userVariables [
  User.bisFootHidden {type=NiagaraBool}
  User.bisHandHidden {type=NiagaraBool}
  User.Damage {type=NiagaraFloat, defaultValue=0, description=""}
  User.SkelMesh {
   type=NiagaraDataInterfaceSkeletalMesh
   defaultValue {
    struct=NiagaraExt_VariableValue_DataInterface
    value {dataInterface=&self:NiagaraDataInterfaceSkeletalMesh_0}
   }
  }
  User.SkelMeshBody {
   type=NiagaraDataInterfaceSkeletalMesh
   defaultValue {
    struct=NiagaraExt_VariableValue_DataInterface
    value {dataInterface=&self:NiagaraDataInterfaceSkeletalMesh_1}
   }
  }
 ]
}
