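# Niagara system for the Party Trooper character's cracked eyes, a damage-reactive sprite effect attached to the skeletal mesh.
# A burst spawns four sprites that follow eye sockets, with custom offset modules shaping the crack, shift and swivel from user damage timing through a ramp curve.
# A timeline curve pulses sprite size, owner scale adjusts camera offset and size, and the head-hidden flag sets the visibility tag.
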NiagaraSystem /BRCosmetics/Characters/Player/Male/Medium/Bodies/M_MED_Party_Trooper_A/FX/NS_M_PartyTrooper_A_CrackEyes
dependencies {
 usedDataInterfaces [&NiagaraDataInterfaceCurve, &NiagaraDataInterfaceArrayFloat]
 usedDynamicInputs [&SetBoolByFloatComparison]
 usedModules [&SystemState, &Timeline]
}
emitters {Eye_Cracked {
  emitterUpdateScript [
   &EmitterState()
   &SpawnBurst_Instantaneous(Spawn Count=4, Use Spawn Probability=false)
  ]
  particleSpawnScript [
   &InitializeParticle(
    Color Mode=::NewEnumerator0
    Lifetime=15
    Position Offset=<0 0 125>
    Sprite Rotation Mode=::NewEnumerator0
    Sprite Size Mode=::NewEnumerator3
    Uniform Sprite Size=8.3
    UsePositionOffset=false
    Write Material Random=true
   )
   &self:Init_EyeOffsets(EyeCrackAngle=-8, EyeCrackShift=0.3) {module=Init_EyeOffsets}
   &SampleSkeletalMesh(
    AdditionalYawPitchRoll=false
    Bone / Socket Index:NiagaraInt32=$Particles.EyeSocketIdx
    Mesh Sampling Type:ENiagaraSkelSamplingModeFull=::NewEnumerator5
    Skeletal Mesh:NiagaraDataInterfaceSkeletalMesh=$User.SkelMesh
    Socket Sampling Mode:ENiagaraSocketSamplingMode=::NewEnumerator5
   )
   &CameraOffset(Camera Offset Amount=&Multiply_Float(A=&MakeFloatFromVector(Vector=$Engine.Owner.Scale, Channel=::NewEnumerator0), B=1))
  ]
  particleUpdateScript [
   &ParticleState()
   &SkeletalMeshLocation(
    Bone / Socket Index=$Particles.EyeSocketIdx
    "Interpolate Mesh Sampling (Update)"=true
    MaskSpawnGroup=false
    Mesh Sampling Type=::NewEnumerator5
    Sampled Position Offset Coordinate Space:ENiagaraSkelMeshTransforms=::NewEnumerator0
    Skeletal Mesh=$User.SkelMesh
    Socket Sampling Mode=::NewEnumerator5
    Use Sampled Mesh Attribute as Ribbon ID=false
   )
   &self:Update_EyeOffsets(
    DamageDuration:NiagaraFloat=$User.DamageDuration
    DamageTime:NiagaraFloat=$User.DamageTime
    DamageTimeRamped:NiagaraFloat=&FloatFromCurve(
     FloatCurve=DataInterface() {propertyValues {Curve {keys [
         {interpMode=RCIM_Cubic, tangentMode=RCTM_User, time=0, value=0}
         {arriveTangent=5.4924, arriveTangentWeight=0.0359, leaveTangent=5.4924, leaveTangentWeight=0.0248, time=0.03, value=1}
         {arriveTangentWeight=0.0607, time=0.06, value=1}
         <at 0.98 1>
         <at 1 0>
        ]}}}
     CurveIndex=$Particles.DamageTimeElapsed
    )
    Eye Halves Swivel Z=-35
    EyeCrackShift=0.7
    EyeOffset:Vector3f=<0 0.15 0>
    "Eyes LR Fold  Z"=-10
    Eyes LR Tilt Y=0
    SM_SampledRotation:Quat4f=$Particles.SkeletalMeshLocation.SampledRotation
   ) {module=Update_EyeOffsets}
   /SetVariables(Particles.SpriteSize=&Multiply_Vector2D_ByFloat(
     Float=&Multiply_Float(
      A=&FloatFromCurve(
       FloatCurve=DataInterface() {propertyValues {Curve {keys [
           {interpMode=RCIM_Cubic, tangentMode=RCTM_User, time=0, value=0.975}
           {arriveTangent=1.0761, interpMode=RCIM_Cubic, leaveTangent=1.0761, tangentMode=RCTM_User, time=0.0004, value=1.4938}
           {arriveTangent=-2.2833, interpMode=RCIM_Cubic, leaveTangent=-2.2833, tangentMode=RCTM_User, time=0.3133, value=0.9728}
           {arriveTangent=-0.1031, interpMode=RCIM_Cubic, leaveTangent=-0.1031, tangentMode=RCTM_User, time=0.4575, value=0.7433}
           <at 0.8027 0.9917>
           {tangentMode=RCTM_User, time=1, value=0.975}
          ]}}}
       CurveIndex=$System.Timeline.NormalizedTime
      )
      B=1
     )
     Vector2D=$Particles.Initial.SpriteSize
    )) {isSetParameters=true}
   &self:DepthFade(DepthFade=0.05) {module=DepthFade}
   &ApplyOwnerScaleToAttributes(Scale Initial Sprite Size=true)
   /SetVariables(Particles.VisibilityTag=&MakeCustomIntFromBool(Bool=$User.bisHeadHidden, True INT=1, False INT=0)) {isSetParameters=true}
  ]
  rendererClasses [&NiagaraSpriteRendererProperties]
  renderers [{rendererClass=&NiagaraSpriteRendererProperties}]
 }}
summary {
 systemName=NS_M_PartyTrooper_A_CrackEyes
 emitters [{emitterName=Eye_Cracked, rendererClasses [&NiagaraSpriteRendererProperties]}]
 userVariables [
  User.bisHeadHidden {type=NiagaraBool}
  User.Damage {type=NiagaraFloat, defaultValue=0, description=""}
  User.DamageDuration {type=NiagaraFloat, defaultValue=0, description=""}
  User.DamageTime {type=NiagaraFloat, defaultValue=0, description=""}
  User.SkelMesh {
   type=NiagaraDataInterfaceSkeletalMesh
   defaultValue {
    struct=NiagaraExt_VariableValue_DataInterface
    value {dataInterface=&self:NiagaraDataInterfaceSkeletalMesh_0}
   }
  }
 ]
}
